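# Niagara idle effect for the LemonCartGranite back bling, drawing short-lived, bright cyan ripples at two skeletal mesh bones or sockets.
# MeshRipples bursts one pair of particles; MeshRipples002 adds three staggered pairs that also grow over life.
# Both use spawn groups to pick each particle's attach point, re-sample it every update, and render as meshes and sprites.
# A custom VibratePosition adds random jitter and per-group drift, rotated by the sampled rotation.
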NiagaraSystem /BRCosmetics/Accessories/FORT_Backpacks/Backpack_M_MED_LemonCartGranite/FX/NS_Backpack_LemonCartGranite_Idle
dependencies {
 usedDataInterfaces [&NiagaraDataInterfaceCurve, &NiagaraDataInterfaceColorCurve, &NiagaraDataInterfaceSkeletalMesh]
 usedDynamicInputs []
 usedModules [
  &SystemState
  &self:SystemSpawnScript.SystemScriptSource.SystemScriptGraph.NiagaraNodeAssignment_2.SetVariables_#1
 ]
}
emitters {
 MeshRipples {
  emitterUpdateScript [
   &EmitterState()
   &SpawnBurst_Instantaneous(Spawn Count=2, Use Spawn Probability=false, UseLoopCountLimit=false)
  ]
  particleSpawnScript [
   &InitializeParticle(
    Lifetime=0.3
    Mesh Scale Mode=::NewEnumerator3
    Mesh Uniform Scale=0.02
    Position Offset=<5 0 -2.5>
    Sprite Size Mode=::NewEnumerator3
    Uniform Sprite Size=8
    UsePositionOffset=false
    Write Material Random=true
   )
   &CameraOffset()
   &FN_SetSpawnGroup(Number of Spawn Groups=2)
   &SkeletalMeshLocation(
    Bone / Socket Index=$Particles.SpawnGroup
    Bone Sampling Mode=::NewEnumerator4
    MaskSpawnGroup=false
    Orientation Sampling=::NewEnumerator1
    Sampled Position Offset=&MaskVectorBySpawnGroup(Masked Vector=<0 0 0>, Passthrough Vector=<0 0 0>)
    Skeletal Mesh=$System.Skeletal Mesh
   ) {module=SkeletalMeshLocation001}
   &InitialMeshOrientation(
    Mesh Orientation Mode=::NewEnumerator2
    Orientation Vector=<0 0 1>
    Rotation=&MaskVectorBySpawnGroup(Masked Vector=<0 0 0>, Passthrough Vector=<0.25 0 0>)
    Use Rotation Vector=true
   ) {module=InitialMeshOrientation001}
   /SetVariables(Particles.SpawnGroup=1) {enabled=false, isSetParameters=true, module=SetVariables_#2}
   &CameraOffset(Camera Offset Amount=2) {module=CameraOffset001}
  ]
  particleUpdateScript [
   &ParticleState()
   &SkeletalMeshLocation(Bone / Socket Index=$Particles.SpawnGroup, Bone Sampling Mode=::NewEnumerator4, MaskSpawnGroup=false, Skeletal Mesh=$System.Skeletal Mesh)
   &ScaleColor(
    Scale Alpha=&FloatFromCurve(
     FloatCurve=DataInterface() {propertyValues {Curve {keys [
         {arriveTangent=-1.1111, interpMode=RCIM_Cubic, time=0, value=0}
         <at 0.1 1 RCIM_Cubic>
         {arriveTangent=-0.9656, interpMode=RCIM_Cubic, leaveTangent=-0.9656, tangentMode=RCTM_User, time=1, value=0}
        ]}}}
     Scale Curve=0.3
    )
    Scale RGB=&Multiply_VectorByFloat(
     Vector=&MakeVectorFromLinearColorRGB(Color=&ColorFromCurve(ColorCurve=DataInterface() {propertyValues {
         AlphaCurve {keys []}
         BlueCurve {keys [
           {interpMode=RCIM_Cubic, leaveTangentWeight=0.8633, time=0, value=1.25}
           {arriveTangentWeight=0.5327, interpMode=RCIM_Cubic, time=1, value=1.25}
          ]}
         GreenCurve {keys [
           {interpMode=RCIM_Cubic, leaveTangentWeight=0.8633, time=0, value=0.8626}
           {
            arriveTangent=6.199
            arriveTangentWeight=0.5327
            interpMode=RCIM_Cubic
            leaveTangent=6.199
            leaveTangentWeight=0.6868
            tangentMode=RCTM_User
            time=1
            value=1.0751
           }
          ]}
         RedCurve {keys [
           {interpMode=RCIM_Cubic, leaveTangentWeight=0.8633, time=0, value=0}
           {
            arriveTangent=6.199
            arriveTangentWeight=0.5327
            interpMode=RCIM_Cubic
            leaveTangent=6.199
            leaveTangentWeight=0.6899
            tangentMode=RCTM_User
            time=1
            value=0.1797
           }
          ]}
        }}))
     Float=4
    )
    ScaleA=true
    ScaleRGB=true
   )
   &DynamicMaterialParameters(
    Index 0 Param 1=3
    Index 0 Param 2=1
    Module.Index0Param0Write=true
    Module.Index0Param1Write=true
    Module.Index0Param2Write=true
    Module.Index0Param3Write=true
   )
   /SetVariables(Particles.VibratePosition:NiagaraPosition=&AddVectorToPosition(
     Position=$Particles.Position
     Vector=&RotateVectorByQuaternion(
      Vector To Rotate=&Add_Vector(
       A=&Multiply_Vector(
        A=&Add_Vector(
         A=&Multiply_Vector(
          A=/Game/Effects/Niagara/DynamicInputs/Engine/FN_GetSystemAxis(Axis:ENiagaraOrientationAxis=::XAxis, Flip Axis:NiagaraBool=false)
          B=&VectorFromFloat(Value=&RandomRangeFloat(Minimum=-1, Maximum=-1, Evaluation Type=::NewEnumerator1))
         )
         B=&Multiply_Vector(
          A=&ReturnMeshOrientationAxis(Mesh Orientation Axis=::YAxis, Mesh Orientation Quat=$Particles.MeshOrientation)
          B=&VectorFromFloat(Value=&RandomRangeFloat(Minimum=-1, Maximum=1, Evaluation Type=::NewEnumerator1))
         )
        )
        B=&VectorFromFloat(Value=0.1)
       )
       B=&MakeVector(
        X=&MaskFloatBySpawnGroup(
         Masked Float=&FloatFromCurve(
          FloatCurve=DataInterface() {propertyValues {Curve {keys [{tangentMode=RCTM_User, time=0, value=0}, {tangentMode=RCTM_User, time=1, value=3}]}}}
          Scale Curve=-1
         )
         Passthrough Float=0
        )
        Y=&MaskFloatBySpawnGroup(
         Masked Float=&FloatFromCurve(
          FloatCurve=DataInterface() {propertyValues {Curve {keys [{tangentMode=RCTM_User, time=0, value=0}, {tangentMode=RCTM_User, time=1, value=3}]}}}
          Scale Curve=-1
         )
         Passthrough Float=0
         Spawn Group Mask=1
        )
        Z=0
       )
      )
      Quaternion=$Particles.SkeletalMeshLocation001.SampledRotation
     )
    )) {isSetParameters=true, module=SetVariables_#3}
   /SetVariables(Particles.VibratePosition:NiagaraPosition=&AddVectorToPosition(
     Position=$Particles.Position
     Vector=&Add_Vector(
      A=&Multiply_Vector(
       A=&Add_Vector(
        A=&Multiply_Vector(
         A=&FN_GetSystemAxis(Axis:ENiagaraOrientationAxis=::XAxis, Flip Axis:NiagaraBool=false)
         B=&VectorFromFloat(Value=&RandomRangeFloat(Minimum=-1, Maximum=-1, Evaluation Type=::NewEnumerator1))
        )
        B=&Multiply_Vector(
         A=&ReturnMeshOrientationAxis(Mesh Orientation Axis=::YAxis, Mesh Orientation Quat=$Particles.MeshOrientation)
         B=&VectorFromFloat(Value=&RandomRangeFloat(Minimum=-1, Maximum=1, Evaluation Type=::NewEnumerator1))
        )
       )
       B=&VectorFromFloat(Value=1)
      )
      B=&MakeVector(
       X=&MaskFloatBySpawnGroup(Masked Float=0, Passthrough Float=1)
       Y=&MaskFloatBySpawnGroup(
        Masked Float=0
        Passthrough Float=&FloatFromCurve(
         FloatCurve=DataInterface() {propertyValues {Curve {keys [{tangentMode=RCTM_User, time=0, value=0}, {tangentMode=RCTM_User, time=1, value=10}]}}}
         Scale Curve=-1
        )
       )
       Z=0
      )
     )
    )) {enabled=false, isSetParameters=true, module=SetVariables_#4}
  ]
  rendererClasses [&NiagaraMeshRendererProperties, &NiagaraSpriteRendererProperties]
  renderers [
   {rendererClass=&NiagaraMeshRendererProperties, rendererIndex=0}
   {rendererClass=&NiagaraMeshRendererProperties, rendererIndex=1}
   {rendererClass=&NiagaraSpriteRendererProperties, rendererIndex=2}
  ]
 }
 MeshRipples002 {
  emitterUpdateScript [
   &EmitterState()
   &SpawnBurst_Instantaneous(Spawn Count=2, Use Spawn Probability=false, UseLoopCountLimit=false)
   &SpawnBurst_Instantaneous(Spawn Count=2, Spawn Time=0.15, Use Spawn Probability=false, UseLoopCountLimit=false) {module=SpawnBurst_Instantaneous003}
   &SpawnBurst_Instantaneous(Spawn Count=2, Spawn Time=0.05, Use Spawn Probability=false, UseLoopCountLimit=false) {module=SpawnBurst_Instantaneous001}
  ]
  particleSpawnScript [
   &InitializeParticle(
    Lifetime=0.25
    Mesh Scale Mode=::NewEnumerator3
    Mesh Uniform Scale=0.02
    Position Offset=<5 0 -2.5>
    Sprite Size Mode=::NewEnumerator3
    Uniform Sprite Size=8
    UsePositionOffset=false
    Write Material Random=true
   )
   &CameraOffset()
   &FN_SetSpawnGroup(Number of Spawn Groups=2)
   &SkeletalMeshLocation(
    Bone / Socket Index=$Particles.SpawnGroup
    Bone Sampling Mode=::NewEnumerator4
    MaskSpawnGroup=false
    Orientation Sampling=::NewEnumerator1
    Sampled Position Offset=&MaskVectorBySpawnGroup(Masked Vector=<0 0 0>, Passthrough Vector=<0 0 0>)
    Skeletal Mesh=$System.Skeletal Mesh
   ) {module=SkeletalMeshLocation001}
   &InitialMeshOrientation(
    Mesh Orientation Mode=::NewEnumerator2
    Orientation Vector=<0 0 1>
    Rotation=&MaskVectorBySpawnGroup(Masked Vector=<0 0 0>, Passthrough Vector=<0.25 0 0>)
    Use Rotation Vector=true
   ) {module=InitialMeshOrientation001}
   /SetVariables(Particles.SpawnGroup=1) {enabled=false, isSetParameters=true, module=SetVariables_#2}
   &CameraOffset(Camera Offset Amount=2) {module=CameraOffset001}
  ]
  particleUpdateScript [
   &ParticleState()
   &SkeletalMeshLocation(Bone / Socket Index=$Particles.SpawnGroup, Bone Sampling Mode=::NewEnumerator4, MaskSpawnGroup=false, Skeletal Mesh=$System.Skeletal Mesh)
   &ScaleMeshSize(Scale Factor=&VectorFromFloat(Value=&Add_Float(
      A=1
      B=&FloatFromCurve(
       FloatCurve=DataInterface() {propertyValues {Curve {keys [
           {
            arriveTangent=1.0578
            arriveTangentWeight=0.2235
            interpMode=RCIM_Cubic
            leaveTangent=1.0578
            leaveTangentWeight=0.2235
            tangentMode=RCTM_User
            time=0.05
            value=0
           }
           {
            arriveTangent=5.2964
            arriveTangentWeight=0.9219
            interpMode=RCIM_Cubic
            leaveTangent=5.2964
            leaveTangentWeight=0.9219
            tangentMode=RCTM_User
            time=1
            value=2
           }
          ]}}}
       Scale Curve=&RandomRangeFloat(Minimum=0.5, Maximum=1)
      )
     )))
   &ScaleColor(
    Scale Alpha=&FloatFromCurve(
     FloatCurve=DataInterface() {propertyValues {Curve {keys [
         {arriveTangent=-1.1111, interpMode=RCIM_Cubic, time=0, value=0}
         <at 0.1 1 RCIM_Cubic>
         {arriveTangent=-0.9656, interpMode=RCIM_Cubic, leaveTangent=-0.9656, tangentMode=RCTM_User, time=1, value=0}
        ]}}}
     Scale Curve=3
    )
    Scale RGB=&Multiply_VectorByFloat(
     Vector=&MakeVectorFromLinearColorRGB(Color=&ColorFromCurve(ColorCurve=DataInterface() {propertyValues {
         AlphaCurve {keys []}
         BlueCurve {keys [
           {interpMode=RCIM_Cubic, leaveTangentWeight=0.8633, time=0, value=1.25}
           {arriveTangentWeight=0.5327, interpMode=RCIM_Cubic, time=1, value=1.25}
          ]}
         GreenCurve {keys [
           {interpMode=RCIM_Cubic, leaveTangentWeight=0.8633, time=0, value=0.8626}
           {
            arriveTangent=6.199
            arriveTangentWeight=0.5327
            interpMode=RCIM_Cubic
            leaveTangent=6.199
            leaveTangentWeight=0.6868
            tangentMode=RCTM_User
            time=1
            value=1.0751
           }
          ]}
         RedCurve {keys [
           {interpMode=RCIM_Cubic, leaveTangentWeight=0.8633, time=0, value=0}
           {
            arriveTangent=6.199
            arriveTangentWeight=0.5327
            interpMode=RCIM_Cubic
            leaveTangent=6.199
            leaveTangentWeight=0.6899
            tangentMode=RCTM_User
            time=1
            value=0.1797
           }
          ]}
        }}))
     Float=4
    )
    ScaleA=true
    ScaleRGB=true
   )
   &DynamicMaterialParameters(
    Index 0 Param 4=&FloatFromCurve(
     FloatCurve=DataInterface() {propertyValues {Curve {keys [
         {
          arriveTangent=9.177
          arriveTangentWeight=0.6327
          interpMode=RCIM_Cubic
          leaveTangent=9.177
          leaveTangentWeight=0.6327
          tangentMode=RCTM_User
          tangentWeightMode=RCTWM_WeightedBoth
          time=0
          value=0.2
         }
         {
          arriveTangentWeight=0.6785
          interpMode=RCIM_Cubic
          leaveTangentWeight=0.6785
          tangentMode=RCTM_User
          tangentWeightMode=RCTWM_WeightedBoth
          time=1
          value=1
         }
        ]}}}
     Scale Curve=0.015
    )
    Module.Index0Param0Write=true
    Module.Index0Param1Write=true
    Module.Index0Param2Write=true
    Module.Index0Param3Write=true
   )
   /SetVariables(Particles.VibratePosition:NiagaraPosition=&AddVectorToPosition(
     Position=$Particles.Position
     Vector=&RotateVectorByQuaternion(
      Vector To Rotate=&Add_Vector(
       A=&Multiply_Vector(
        A=&Add_Vector(
         A=&Multiply_Vector(
          A=&FN_GetSystemAxis(Axis:ENiagaraOrientationAxis=::XAxis, Flip Axis:NiagaraBool=false)
          B=&VectorFromFloat(Value=&RandomRangeFloat(Minimum=-1, Maximum=-1, Evaluation Type=::NewEnumerator1))
         )
         B=&Multiply_Vector(
          A=&ReturnMeshOrientationAxis(Mesh Orientation Axis=::YAxis, Mesh Orientation Quat=$Particles.MeshOrientation)
          B=&VectorFromFloat(Value=&RandomRangeFloat(Minimum=-1, Maximum=1, Evaluation Type=::NewEnumerator1))
         )
        )
        B=&VectorFromFloat(Value=0.1)
       )
       B=&MakeVector(
        X=&MaskFloatBySpawnGroup(
         Masked Float=&FloatFromCurve(
          FloatCurve=DataInterface() {propertyValues {Curve {keys [{tangentMode=RCTM_User, time=0, value=0}, {tangentMode=RCTM_User, time=1, value=3}]}}}
          Scale Curve=-1
         )
         Passthrough Float=0
        )
        Y=&MaskFloatBySpawnGroup(
         Masked Float=&FloatFromCurve(
          FloatCurve=DataInterface() {propertyValues {Curve {keys [{tangentMode=RCTM_User, time=0, value=0}, {tangentMode=RCTM_User, time=1, value=3}]}}}
          Scale Curve=-1
         )
         Passthrough Float=0
         Spawn Group Mask=1
        )
        Z=0
       )
      )
      Quaternion=$Particles.SkeletalMeshLocation001.SampledRotation
     )
    )) {isSetParameters=true, module=SetVariables_#3}
   /SetVariables(Particles.VibratePosition:NiagaraPosition=&AddVectorToPosition(
     Position=$Particles.Position
     Vector=&Add_Vector(
      A=&Multiply_Vector(
       A=&Add_Vector(
        A=&Multiply_Vector(
         A=&FN_GetSystemAxis(Axis:ENiagaraOrientationAxis=::XAxis, Flip Axis:NiagaraBool=false)
         B=&VectorFromFloat(Value=&RandomRangeFloat(Minimum=-1, Maximum=-1, Evaluation Type=::NewEnumerator1))
        )
        B=&Multiply_Vector(
         A=&ReturnMeshOrientationAxis(Mesh Orientation Axis=::YAxis, Mesh Orientation Quat=$Particles.MeshOrientation)
         B=&VectorFromFloat(Value=&RandomRangeFloat(Minimum=-1, Maximum=1, Evaluation Type=::NewEnumerator1))
        )
       )
       B=&VectorFromFloat(Value=1)
      )
      B=&MakeVector(
       X=&MaskFloatBySpawnGroup(Masked Float=0, Passthrough Float=1)
       Y=&MaskFloatBySpawnGroup(
        Masked Float=0
        Passthrough Float=&FloatFromCurve(
         FloatCurve=DataInterface() {propertyValues {Curve {keys [{tangentMode=RCTM_User, time=0, value=0}, {tangentMode=RCTM_User, time=1, value=10}]}}}
         Scale Curve=-1
        )
       )
       Z=0
      )
     )
    )) {enabled=false, isSetParameters=true, module=SetVariables_#4}
  ]
  rendererClasses [&NiagaraMeshRendererProperties, &NiagaraSpriteRendererProperties]
  renderers [
   {rendererClass=&NiagaraMeshRendererProperties, rendererIndex=0}
   {rendererClass=&NiagaraMeshRendererProperties, rendererIndex=1}
   {rendererClass=&NiagaraSpriteRendererProperties, rendererIndex=2}
  ]
 }
}
summary {
 systemName=NS_Backpack_LemonCartGranite_Idle
 emitters [
  {emitterName=MeshRipples002, rendererClasses [&NiagaraMeshRendererProperties, &NiagaraSpriteRendererProperties]}
  {emitterName=MeshRipples, rendererClasses [&NiagaraMeshRendererProperties, &NiagaraSpriteRendererProperties]}
 ]
 userVariables [
  User.bIsFrontEnd {type=NiagaraBool}
  User.bIsFrontEndPreview {type=NiagaraBool}
 ]
}
