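# Parent material for the LemonCartGranite backpack that draws animated, glowing eyes over the standard 2023 character material.
# The eyes sample a 3x3 flipbook atlas with placed, scaled UVs and bump-offset parallax.
# Time-driven vector noise switches between sine-cycled frames with brief periodic drops to frame 0 (blink-like) and noise-picked special poses.
# The masked atlas blends Eyes_Color into Emissive between the CharacterParent PreFX and PostFX functions.
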Material /BRCosmetics/Accessories/FORT_Backpacks/Backpack_M_MED_LemonCartGranite/Materials/M_LemonCartGranite_Backpack_Parent
graph {
 properties {
  bHasPixelAnimation=true
  bUseMaterialAttributes=true
  bUsedWithMeshDeformer=true
  bUsedWithMorphTargets=true
  bUsedWithSkeletalMesh=true
  pixelDepthOffsetMode=PDOM_Legacy
  previewMesh=/BRCosmetics/Accessories/FORT_Backpacks/Backpack_M_MED_LemonCartGranite/Meshes/M_MED_LemonCartGranite_Pack
  shadingModel=MSM_FromMaterialExpression
  usedShadingModels=Default Lit | Subsurface | Clear Coat | Subsurface Profile | Cloth
 }
 nodes [
  n1 = MaterialFunctionCall() {materialFunction=&MF_CharacterParent_2023_PreFX}
  n2 = GetMaterialAttributes(None=@n1.Result) {attributeGetTypes [#1]}
  n3 = VectorParameter() {parameterName=Eyes_Color, defaultValue=<rgba 0.9 0.9 0.9 1>}
  n4 = ScalarParameter() {defaultValue=0.9608, parameterName=Eyes_U, sliderMax=1}
  n5 = ScalarParameter() {defaultValue=0.5328, parameterName=Eyes_V, sliderMax=1}
  n6 = MaterialFunctionCall!(
   Texture Scale=ScalarParameter() {defaultValue=0.1, parameterName=Eyes_Scale}
   UVs
   Offset XY Texturesize Z=AppendVector(A=AppendVector(A=@n4, B=@n5), B=Constant() {r=1})
  ) {materialFunction=/Engine/Functions/Engine_MaterialFunctions01/Texturing/ScaleUVsAroundPoint}
  n7 = MaterialFunctionCall(Float2=BumpOffset(Coordinate=@n6.UVs, Height=Constant(), HeightRatioInput=ScalarParameter() {defaultValue=0.05, parameterName=Eyes_Depth, sliderMax=1}) {referencePlane=1}) {materialFunction=&BreakOutFloat2Components}
  n8 = MaterialFunctionCall(X=@n7.G, Y=OneMinus(None=@n7.R)) {materialFunction=&MakeFloat2}
  n9 = ScalarParameter() {defaultValue=0.6, parameterName=Eyes_SpecialPoseThreshold}
  n10 = Time()
  n11 = ScalarParameter() {defaultValue=0.4, parameterName=Eyes_AnimSpeed}
  n12 = MaterialFunctionCall(Float3=VectorNoise(World Position=Multiply(A=@n10, B=@n11) {constB=0.15})) {materialFunction=&BreakOutFloat3Components}
  n13 = MaterialFunctionCall(Speed=Constant() {r=0.3}) {materialFunction=&DebugTimeSine}
  n14 = Constant() {r=0.3}
  n15 = Constant() {r=0.17}
  n16 = Min(A=Frac(None=Multiply(A=@n10, B=@n14)), B=Frac(None=Multiply(A=@n10, B=@n15)))
  n17 = Constant() {r=0.06}
  n18 = Constant()
  n19 = Constant() {r=1}
  n20 = Divide(A=LinearInterpolate(A=LinearInterpolate(Alpha=@n13.Result) {constA=3.1, constB=5.9}, Alpha=If(A=@n16, B=@n17, "A > B"=@n18, "A < B"=@n19)) {constB=0.25}) {constB=9}
  n21 = LinearInterpolate(Alpha=@n12.G) {constA=0.1111, desc=Exclude pose 0}
  n22 = Constant() {r=3}
  n23 = TextureSampleParameter2D(UVs=MaterialFunctionCall("Animation  Phase (0-1)"=If(A=@n9, B=@n12.R, "A > B"=@n20, "A < B"=@n21), Number of Rows=@n22, Number of Columns=@n22, UVs=@n8.Result) {materialFunction=&FlipBook}.UVs) {
   parameterName=Eye Atlas
   samplerSource=SSM_Wrap_WorldGroupSettings
   texture=/BRCosmetics/Accessories/FORT_Backpacks/Backpack_M_MED_LemonCartGranite/Textures/T_LemonCartGranite_Backpack_Eyes
  }
  n24 = Multiply(
   A=Multiply(
    A=@n6.0-1 mask
    B=MaterialFunctionCall(UVs=@n8.Result) {materialFunction=&ScaleUVsByCenter}.0-1 mask
   )
   B=@n23.R
  )
  MaterialFunctionCall(
   Material=SetMaterialAttributes(MaterialAttributes=@n2.MaterialAttributes, Emissive Color=LinearInterpolate(A=@n2.Emissive Color, B=@n3.RGB, Alpha=@n24)) {attributeSetTypes [#1]}
   M=@n1.M
   Pre Skinned Local Position=@n1.M
   Distance Blend=@n1.M
   Fresnel=@n1.M
  ) {materialFunction=&MF_CharacterParent_2023_PostFX}
 ]
}
