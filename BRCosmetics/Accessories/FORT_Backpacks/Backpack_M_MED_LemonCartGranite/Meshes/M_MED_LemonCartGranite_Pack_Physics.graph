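# Physics asset for the LemonCartGranite backpack cosmetic, giving two palm-like branching structures on the pack constrained physical sway.
# Each palm is a three-segment chain from the pack body, whose top segment fans out into many short chains, apparently fronds, mostly three bones long.
# Every body is a small capsule, and symmetric swing and twist limits are tightest along the trunks and loosen slightly toward the tips.
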PhysicsAsset /BRCosmetics/Accessories/FORT_Backpacks/Backpack_M_MED_LemonCartGranite/Meshes/M_MED_LemonCartGranite_Pack_Physics
bodies {
 joint10 [{shapeType=Capsule, length=1, radius=1, shapeName=joint10_capsule, center=<1.8076 0.3929 0.2586>, rotation=<rot 85.0002 179.9997 179.9998>}]
 joint11 [{shapeType=Capsule, length=1, radius=1, shapeName=joint11_capsule, center=<1.221 -0.1728 -0.0656>, rotation=<rot 90 0 0>}]
 joint12 [{shapeType=Capsule, length=1, radius=1, shapeName=joint12_capsule, center=<1.1453 0.019 -0.8803>, rotation=<rot 90 0 0>}]
 joint13 [{shapeType=Capsule, length=1, radius=1, shapeName=joint13_capsule, center=<1.8271 -0.3295 -0.0561>, rotation=<rot 84.9999 179.9981 179.9983>}]
 joint14 [{shapeType=Capsule, length=1, radius=1, shapeName=joint14_capsule, center=<1.8243 0.1652 -0.0666>, rotation=<rot 90 0 0>}]
 joint15 [{shapeType=Capsule, length=1, radius=1, shapeName=joint15_capsule, center=<1.1542 -0.1115 0.7369>, rotation=<rot 90 0 0>}]
 joint16 [{shapeType=Capsule, length=1, radius=1, shapeName=joint16_capsule, center=<1.5934 -0.1206 -0.1492>, rotation=<rot 85 179.9979 179.998>}]
 joint17 [{shapeType=Capsule, length=1, radius=1, shapeName=joint17_capsule, center=<1.4027 0.0334 -0.0625>, rotation=<rot 90 0 0>}]
 joint18 [{shapeType=Capsule, length=1, radius=1, shapeName=joint18_capsule, center=<0.861 -0.0202 0.5561>, rotation=<rot 90 0 0>}]
 joint19 [{shapeType=Capsule, length=1, radius=1, shapeName=joint19_capsule, center=<1.9825 0.0029 0.1708>, rotation=<rot 85.0001 179.9983 179.9985>}]
 joint20 [{shapeType=Capsule, length=1, radius=1, shapeName=joint20_capsule, center=<1.529 -0.0346 0.0771>, rotation=<rot 90 0 0>}]
 joint21 [{shapeType=Capsule, length=1, radius=1, shapeName=joint21_capsule, center=<1.5656 -0.7039 0.6638>, rotation=<rot 90 0 0>}]
 joint22 [{shapeType=Capsule, length=1, radius=1, shapeName=joint22_capsule, center=<2.5186 0.229 0.0487>, rotation=<rot 85.0002 179.9991 179.9991>}]
 joint23 [{shapeType=Capsule, length=1, radius=1, shapeName=joint23_capsule, center=<1.5258 -0.0089 0.0499>, rotation=<rot 90 0 0>}]
 joint24 [{shapeType=Capsule, length=1, radius=1, shapeName=joint24_capsule, center=<0.7953 0.1879 0.4517>, rotation=<rot 90 0 0>}]
 joint25 [{shapeType=Capsule, length=1, radius=1, shapeName=joint25_capsule, center=<1.8589 0.2356 -0.3026>, rotation=<rot 84.9999 -180 -179.9995>}]
 joint26 [{shapeType=Capsule, length=1, radius=1, shapeName=joint26_capsule, center=<1.2612 -0.2719 -0.3516>, rotation=<rot 90 0 0>}]
 joint27 [{shapeType=Capsule, length=1, radius=1, shapeName=joint27_capsule, center=<1.0327 -0.21 -1.2152>, rotation=<rot 90 0 0>}]
 joint28 [{shapeType=Capsule, length=1, radius=1, shapeName=joint28_capsule, center=<1.3096 -0.0658 0.0498>, rotation=<rot 84.9998 -179.999 -179.9991>}]
 joint29 [{shapeType=Capsule, length=1, radius=1, shapeName=joint29_capsule, center=<1.4009 0.1822 -0.0331>, rotation=<rot 90 0 0>}]
 joint30 [{shapeType=Capsule, length=1, radius=1, shapeName=joint30_capsule, center=<1.0161 -0.0342 -0.5807>, rotation=<rot 90 0 0>}]
 joint34 [{shapeType=Capsule, length=1, radius=1, shapeName=joint34_capsule, center=<2.054 0.1715 -0.2753>, rotation=<rot 90 0 0>}]
 joint35 [{shapeType=Capsule, length=1, radius=1, shapeName=joint35_capsule, center=<1.8211 -0.2216 -0.3266>, rotation=<rot 90 0 0>}]
 joint36 [{shapeType=Capsule, length=1, radius=1, shapeName=joint36_capsule, center=<1.5122 0.0759 -1.6383>, rotation=<rot 90 0 0>}]
 joint37 [{shapeType=Capsule, length=1, radius=1, shapeName=joint37_capsule, center=<2.0515 0.5473 0.0541>, rotation=<rot 90 0 0>}]
 joint38 [{shapeType=Capsule, length=1, radius=1, shapeName=joint38_capsule, center=<1.7784 -0.1168 0.0351>, rotation=<rot 90 0 0>}]
 joint39 [{shapeType=Capsule, length=1, radius=1, shapeName=joint39_capsule, center=<1.8686 -0.1863 -1.1104>, rotation=<rot 90 0 0>}]
 joint4 [{shapeType=Capsule, length=1, radius=1, shapeName=joint4_capsule, center=<1.8598 -0.133 0.0039>, rotation=<rot 85 -179.9973 -179.9977>}]
 joint40 [{shapeType=Capsule, length=1, radius=1, shapeName=joint40_capsule, center=<0.9322 -0.2866 -0.1216>, rotation=<rot 90 0 0>}]
 joint41 [{shapeType=Capsule, length=1, radius=1, shapeName=joint41_capsule, center=<1.7278 -0.1845 -0.1792>, rotation=<rot 90 0 0>}]
 joint42 [{shapeType=Capsule, length=1, radius=1, shapeName=joint42_capsule, center=<1.33 0.1155 -1.1889>, rotation=<rot 90 0 0>}]
 joint43 [{shapeType=Capsule, length=1, radius=1, shapeName=joint43_capsule, center=<1.0644 -0.2605 0.5586>, rotation=<rot 90 0 0>}]
 joint44 [{shapeType=Capsule, length=1, radius=1, shapeName=joint44_capsule, center=<1.8892 0.0949 0.5951>, rotation=<rot 90 0 0>}]
 joint45 [{shapeType=Capsule, length=1, radius=1, shapeName=joint45_capsule, center=<0.67 0.0646 1.744>, rotation=<rot 90 0 0>}]
 joint46 [{shapeType=Capsule, length=1, radius=1, shapeName=joint46_capsule, center=<1.9596 -0.1359 0.4836>, rotation=<rot 90 0 0>}]
 joint47 [{shapeType=Capsule, length=1, radius=1, shapeName=joint47_capsule, center=<1.7279 0.3521 0.4577>, rotation=<rot 90 0 0>}]
 joint48 [{shapeType=Capsule, length=1, radius=1, shapeName=joint48_capsule, center=<1.1977 0.0656 1.8529>, rotation=<rot 90 0 0>}]
 joint49 [{shapeType=Capsule, length=1, radius=1, shapeName=joint49_capsule, center=<1.7647 -0.3133 -0.0921>, rotation=<rot 90 0 0>}]
 joint5 [{shapeType=Capsule, length=1, radius=1, shapeName=joint5_capsule, center=<1.5616 -0.1643 0.0437>, rotation=<rot 90 0 0>}]
 joint50 [{shapeType=Capsule, length=1, radius=1, shapeName=joint50_capsule, center=<1.4325 -0.3994 0.374>, rotation=<rot 90 0 0>}]
 joint51 [{shapeType=Capsule, length=1, radius=1, shapeName=joint51_capsule, center=<1.9101 -0.1615 1.8307>, rotation=<rot 90 0 0>}]
 joint52 [{shapeType=Capsule, length=1, radius=1, shapeName=joint52_capsule, center=<1.5766 0.6398 0.1358>, rotation=<rot 90 0 0>}]
 joint53 [{shapeType=Capsule, length=1, radius=1, shapeName=joint53_capsule, center=<1.5748 -0.0269 0.1505>, rotation=<rot 90 0 0>}]
 joint54 [{shapeType=Capsule, length=1, radius=1, shapeName=joint54_capsule, center=<1.8109 0.0284 1.5646>, rotation=<rot 90 0 0>}]
 joint55 [{shapeType=Capsule, length=1, radius=1, shapeName=joint55_capsule, center=<0 1.1289 0>, rotation=<rot 90 0 0>}]
 joint56 [{shapeType=Capsule, length=1, radius=1, shapeName=joint56_capsule, center=<0.9351 0.0004 -0.1189>, rotation=<rot 90 0 0>}]
 joint57 [{shapeType=Capsule, length=1, radius=1, shapeName=joint57_capsule, center=<0.7886 -0.01 0.6022>, rotation=<rot 90 0 0>}]
 joint58 [{shapeType=Capsule, length=1, radius=1, shapeName=joint58_capsule, center=<1.6761 0.8737 -0.3195>, rotation=<rot 90 0 0>}]
 joint59 [{shapeType=Capsule, length=1, radius=1, shapeName=joint59_capsule, center=<1.9482 -0.0968 -0.2987>, rotation=<rot 90 0 0>}]
 joint6 [{shapeType=Capsule, length=1, radius=1, shapeName=joint6_capsule, center=<1.1177 0.2506 -0.3327>, rotation=<rot 90 0 0>}]
 joint60 [{shapeType=Capsule, length=1, radius=1, shapeName=joint60_capsule, center=<1.1238 0.1085 -1.288>, rotation=<rot 90 0 0>}]
 joint61 [{shapeType=Capsule, length=1, radius=1, shapeName=joint61_capsule, center=<1.1325 0.7777 -0.3383>, rotation=<rot 90 0 0>}]
 joint62 [{shapeType=Capsule, length=1, radius=1, shapeName=joint62_capsule, center=<1.5981 -0.054 -0.2511>, rotation=<rot 90 0 0>}]
 joint63 [{shapeType=Capsule, length=1, radius=1, shapeName=joint63_capsule, center=<1.1766 -0.1868 -1.1928>, rotation=<rot 90 0 0>}]
 joint64 [{shapeType=Capsule, length=1, radius=1, shapeName=joint64_capsule, center=<1.3449 0.2691 -0.0943>, rotation=<rot 90 0 0>}]
 joint65 [{shapeType=Capsule, length=1, radius=1, shapeName=joint65_capsule, center=<1.2998 0.0873 -0.0405>, rotation=<rot 90 0 0>}]
 joint66 [{shapeType=Capsule, length=1, radius=1, shapeName=joint66_capsule, center=<0.9842 -0.4078 -0.4045>, rotation=<rot 90 0 0>}]
 joint7 [{shapeType=Capsule, length=1, radius=1, shapeName=joint7_capsule, center=<1.6204 -0.0838 -0.0407>, rotation=<rot 85.0001 -179.9983 -179.9983>}]
 joint9 [{shapeType=Capsule, length=1, radius=1, shapeName=joint9_capsule, center=<1.0138 0.2869 -0.471>, rotation=<rot 90 0 0>}]
 pack [{shapeType=Capsule, length=1, radius=1}]
 palm_01_base_01_jnt [{shapeType=Capsule, length=2, radius=2, shapeName=palm_01_base_01_jnt_capsule, center=<3.3235 -0.0373 -0.386>, rotation=<rot 90 0 0>}]
 palm_01_base_02_jnt [{shapeType=Capsule, length=2, radius=2, shapeName=palm_01_base_02_jnt_capsule, center=<6.1805 -0.0436 0.4553>, rotation=<rot 90 0 0>}]
 palm_01_base_03_jnt [{shapeType=Capsule, length=2, radius=2, shapeName=palm_01_base_03_jnt_capsule, center=<3.2171 -0.012 1.2939>, rotation=<rot 90 0 0>}]
 palm_02_base_01_jnt [{shapeType=Capsule, length=2, radius=2, shapeName=palm_02_base_01_jnt_capsule, center=<2.6224 0.0043 0.1613>, rotation=<rot 90 0 0>}]
 palm_02_base_02_jnt [{shapeType=Capsule, length=2, radius=2, shapeName=palm_02_base_02_jnt_capsule, center=<1.8132 0.0071 0.0907>, rotation=<rot 90 0 0>}]
 palm_02_base_03_jnt [{shapeType=Capsule, length=2, radius=2, shapeName=palm_02_base_03_jnt_capsule, center=<2.7574 0.1032 -0.5821>, rotation=<rot 90 0 0>}]
}
constraints [
 {bone1Name=joint19, bone2Name=palm_01_base_03_jnt, swing1LimitDegrees=7, swing2LimitDegrees=7, twistLimitDegrees=7}
 {bone1Name=joint20, bone2Name=joint19, swing1LimitDegrees=9, swing2LimitDegrees=9, twistLimitDegrees=9}
 {bone1Name=joint21, bone2Name=joint20, swing1LimitDegrees=9, swing2LimitDegrees=9, twistLimitDegrees=9}
 {bone1Name=joint27, bone2Name=joint26, swing1LimitDegrees=9, swing2LimitDegrees=9, twistLimitDegrees=9}
 {bone1Name=joint25, bone2Name=palm_01_base_03_jnt, swing1LimitDegrees=7, swing2LimitDegrees=7, twistLimitDegrees=7}
 {bone1Name=joint26, bone2Name=joint25, swing1LimitDegrees=9, swing2LimitDegrees=9, twistLimitDegrees=9}
 {bone1Name=joint9, bone2Name=joint7, swing1LimitDegrees=9, swing2LimitDegrees=9, twistLimitDegrees=9}
 {bone1Name=joint7, bone2Name=palm_01_base_03_jnt, swing1LimitDegrees=7, swing2LimitDegrees=7, twistLimitDegrees=7}
 {bone1Name=joint12, bone2Name=joint11, swing1LimitDegrees=9, swing2LimitDegrees=9, twistLimitDegrees=9}
 {bone1Name=joint10, bone2Name=palm_01_base_03_jnt, swing1LimitDegrees=7, swing2LimitDegrees=7, twistLimitDegrees=7}
 {bone1Name=joint11, bone2Name=joint10, swing1LimitDegrees=9, swing2LimitDegrees=9, twistLimitDegrees=9}
 {bone1Name=joint15, bone2Name=joint14, swing1LimitDegrees=9, swing2LimitDegrees=9, twistLimitDegrees=9}
 {bone1Name=joint14, bone2Name=joint13, swing1LimitDegrees=9, swing2LimitDegrees=9, twistLimitDegrees=9}
 {bone1Name=joint13, bone2Name=palm_01_base_03_jnt, swing1LimitDegrees=7, swing2LimitDegrees=7, twistLimitDegrees=7}
 {bone1Name=joint18, bone2Name=joint17, swing1LimitDegrees=9, swing2LimitDegrees=9, twistLimitDegrees=9}
 {bone1Name=joint17, bone2Name=joint16, swing1LimitDegrees=9, swing2LimitDegrees=9, twistLimitDegrees=9}
 {bone1Name=joint16, bone2Name=palm_01_base_03_jnt, swing1LimitDegrees=7, swing2LimitDegrees=7, twistLimitDegrees=7}
 {bone1Name=joint30, bone2Name=joint29, swing1LimitDegrees=9, swing2LimitDegrees=9, twistLimitDegrees=9}
 {bone1Name=joint29, bone2Name=joint28, swing1LimitDegrees=9, swing2LimitDegrees=9, twistLimitDegrees=9}
 {bone1Name=joint28, bone2Name=palm_01_base_03_jnt, swing1LimitDegrees=7, swing2LimitDegrees=7, twistLimitDegrees=7}
 {bone1Name=joint24, bone2Name=joint23, swing1LimitDegrees=9, swing2LimitDegrees=9, twistLimitDegrees=9}
 {bone1Name=joint23, bone2Name=joint22, swing1LimitDegrees=9, swing2LimitDegrees=9, twistLimitDegrees=9}
 {bone1Name=joint22, bone2Name=palm_01_base_03_jnt, swing1LimitDegrees=7, swing2LimitDegrees=7, twistLimitDegrees=7}
 {bone1Name=joint6, bone2Name=joint5, swing1LimitDegrees=9, swing2LimitDegrees=9, twistLimitDegrees=9}
 {bone1Name=joint5, bone2Name=joint4, swing1LimitDegrees=9, swing2LimitDegrees=9, twistLimitDegrees=9}
 {bone1Name=joint4, bone2Name=palm_01_base_03_jnt, swing1LimitDegrees=7, swing2LimitDegrees=7, twistLimitDegrees=7}
 {bone1Name=joint61, bone2Name=palm_02_base_03_jnt, swing1LimitDegrees=8, swing2LimitDegrees=8, twistLimitDegrees=8}
 {bone1Name=joint62, bone2Name=joint61, swing1LimitDegrees=8, swing2LimitDegrees=8, twistLimitDegrees=8}
 {bone1Name=joint63, bone2Name=joint62, swing1LimitDegrees=9, swing2LimitDegrees=9, twistLimitDegrees=9}
 {bone1Name=joint58, bone2Name=palm_02_base_03_jnt, swing1LimitDegrees=8, swing2LimitDegrees=8, twistLimitDegrees=8}
 {bone1Name=joint59, bone2Name=joint58, swing1LimitDegrees=8, swing2LimitDegrees=8, twistLimitDegrees=8}
 {bone1Name=joint60, bone2Name=joint59, swing1LimitDegrees=9, swing2LimitDegrees=9, twistLimitDegrees=9}
 {bone1Name=joint64, bone2Name=palm_02_base_03_jnt, swing1LimitDegrees=8, swing2LimitDegrees=8, twistLimitDegrees=8}
 {bone1Name=joint65, bone2Name=joint64, swing1LimitDegrees=8, swing2LimitDegrees=8, twistLimitDegrees=8}
 {bone1Name=joint66, bone2Name=joint65, swing1LimitDegrees=9, swing2LimitDegrees=9, twistLimitDegrees=9}
 {bone1Name=joint40, bone2Name=palm_02_base_03_jnt, swing1LimitDegrees=8, swing2LimitDegrees=8, twistLimitDegrees=8}
 {bone1Name=joint41, bone2Name=joint40, swing1LimitDegrees=8, swing2LimitDegrees=8, twistLimitDegrees=8}
 {bone1Name=joint42, bone2Name=joint41, swing1LimitDegrees=9, swing2LimitDegrees=9, twistLimitDegrees=9}
 {bone1Name=joint43, bone2Name=palm_02_base_03_jnt, swing1LimitDegrees=8, swing2LimitDegrees=8, twistLimitDegrees=8}
 {bone1Name=joint44, bone2Name=joint43, swing1LimitDegrees=8, swing2LimitDegrees=8, twistLimitDegrees=8}
 {bone1Name=joint45, bone2Name=joint44, swing1LimitDegrees=9, swing2LimitDegrees=9, twistLimitDegrees=9}
 {bone1Name=joint46, bone2Name=palm_02_base_03_jnt, swing1LimitDegrees=8, swing2LimitDegrees=8, twistLimitDegrees=8}
 {bone1Name=joint47, bone2Name=joint46, swing1LimitDegrees=8, swing2LimitDegrees=8, twistLimitDegrees=8}
 {bone1Name=joint48, bone2Name=joint47, swing1LimitDegrees=9, swing2LimitDegrees=9, twistLimitDegrees=9}
 {bone1Name=joint52, bone2Name=palm_02_base_03_jnt, swing1LimitDegrees=8, swing2LimitDegrees=8, twistLimitDegrees=8}
 {bone1Name=joint53, bone2Name=joint52, swing1LimitDegrees=8, swing2LimitDegrees=8, twistLimitDegrees=8}
 {bone1Name=joint54, bone2Name=joint53, swing1LimitDegrees=9, swing2LimitDegrees=9, twistLimitDegrees=9}
 {bone1Name=joint55, bone2Name=palm_02_base_03_jnt, swing1LimitDegrees=8, swing2LimitDegrees=8, twistLimitDegrees=8}
 {bone1Name=joint56, bone2Name=joint55, swing1LimitDegrees=8, swing2LimitDegrees=8, twistLimitDegrees=8}
 {bone1Name=joint57, bone2Name=joint56, swing1LimitDegrees=9, swing2LimitDegrees=9, twistLimitDegrees=9}
 {bone1Name=joint49, bone2Name=palm_02_base_03_jnt, swing1LimitDegrees=8, swing2LimitDegrees=8, twistLimitDegrees=8}
 {bone1Name=joint50, bone2Name=joint49, swing1LimitDegrees=8, swing2LimitDegrees=8, twistLimitDegrees=8}
 {bone1Name=joint51, bone2Name=joint50, swing1LimitDegrees=9, swing2LimitDegrees=9, twistLimitDegrees=9}
 {bone1Name=joint37, bone2Name=palm_02_base_03_jnt, swing1LimitDegrees=8, swing2LimitDegrees=8, twistLimitDegrees=8}
 {bone1Name=joint38, bone2Name=joint37, swing1LimitDegrees=8, swing2LimitDegrees=8, twistLimitDegrees=8}
 {bone1Name=joint39, bone2Name=joint38, swing1LimitDegrees=9, swing2LimitDegrees=9, twistLimitDegrees=9}
 {bone1Name=joint34, bone2Name=palm_02_base_03_jnt, swing1LimitDegrees=8, swing2LimitDegrees=8, twistLimitDegrees=8}
 {bone1Name=joint35, bone2Name=joint34, swing1LimitDegrees=8, swing2LimitDegrees=8, twistLimitDegrees=8}
 {bone1Name=joint36, bone2Name=joint35, swing1LimitDegrees=9, swing2LimitDegrees=9, twistLimitDegrees=9}
 {bone1Name=palm_02_base_01_jnt, bone2Name=pack, swing1LimitDegrees=2, swing2LimitDegrees=2, twistLimitDegrees=2}
 {bone1Name=palm_02_base_02_jnt, bone2Name=palm_02_base_01_jnt, swing1LimitDegrees=2, swing2LimitDegrees=2, twistLimitDegrees=2}
 {bone1Name=palm_02_base_03_jnt, bone2Name=palm_02_base_02_jnt, swing1LimitDegrees=2, swing2LimitDegrees=2, twistLimitDegrees=2}
 {bone1Name=palm_01_base_01_jnt, bone2Name=pack, swing1LimitDegrees=2, swing2LimitDegrees=2, twistLimitDegrees=2}
 {bone1Name=palm_01_base_02_jnt, bone2Name=palm_01_base_01_jnt, swing1LimitDegrees=2, swing2LimitDegrees=2, twistLimitDegrees=2}
 {bone1Name=palm_01_base_03_jnt, bone2Name=palm_01_base_02_jnt, swing1LimitDegrees=2, swing2LimitDegrees=2, twistLimitDegrees=2}
]
