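# Animation Blueprint for the LemonCartGranite cosmetic backpack, animating its palms, speakers, bowl and bone.
# A cached rigid-body pose with palms constrained to general_jnt feeds additive bone edits for spine-countering gimbal rotation and backward-parachute tilt.
# Event-graph sine waves drive palm sway and bouncing, with bone-driven controllers adding squash to the speakers.
# Pack offsets for melee, crouch, skydive and other states, plus a look-at chain, are disconnected from the output pose.
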AnimBlueprint /BRCosmetics/Accessories/FORT_Backpacks/Backpack_M_MED_LemonCartGranite/Meshes/M_MED_LemonCartGranite_Pack_AnimBP
target_skeleton=&/M_MED_LemonCartGranite_Pack_Skeleton
graphs [
 AnimGraph {nodes [
   Local Space Ref Pose()
   n1 = "Transform (Modify) Bone - Bone: woOrient_aim_jnt"(Alpha=0.8, Translation=Get worldPos_value() -> (worldPos_value:Vector), Rotation="0,0,0", Scale="1,1,1") {
    rotationMode=BMM_Additive
    rotationSpace=BCS_WorldSpace
    translationMode=BMM_Additive
    translationSpace=BCS_WorldSpace
    alphaBoolBlend {blendOption=Linear}
    boneToModify {boneName=woOrient_aim_jnt}
   }
   n2 = Look At - Bone: general_jnt!(ComponentPose=@n1, bAlphaBoolEnabled:bool=True) -> (Pose) {
    alphaInputType=Bool
    bUseLookUpAxis=true
    interpolationTime=10
    interpolationTriggerThreashold=10
    alphaBoolBlend {blendInTime=0.1, blendOption=Linear, blendOutTime=80}
    boneToModify {boneName=general_jnt}
    lookAtTarget {boneReference {boneName=woOrient_aim_jnt}}
   }
   "Transform (Modify) Bone - Bone: general_jnt"!(ComponentPose=@n2, bAlphaBoolEnabled:bool=True, Translation="0,0,0", Rotation="0,0,100", Scale="1,1,1") -> (Pose) {
    alphaInputType=Bool
    rotationMode=BMM_Additive
    translationSpace=BCS_BoneSpace
    alphaBoolBlend {blendInTime=0.1, blendOption=Linear, blendOutTime=80}
    boneToModify {boneName=general_jnt}
    rotation {roll=100}
   }
   n3 = "Transform (Modify) Bone - Bone: pack"!(
    ComponentPose=Local To Component(LocalPose=Use cached pose 'baseDyn_cache'())
    bAlphaBoolEnabled:bool=Get bIsPlayingMeleeAnim()
    Translation_X:double=0
    Translation_Y:double=Get compTransfWhenMeleeAnimY() -> (compTransfWhenMeleeAnimY:double)
    Translation_Z:double=Get compTransfWhenMeleeAnimZ() -> (compTransfWhenMeleeAnimZ:double)
    Rotation_Roll:float="Get compRotZ_WhenMeleeAnim (Yaw)"() -> ("compRotZ_WhenMeleeAnim (Yaw)":float)
    Rotation_Pitch:float=0
    Rotation_Yaw:float=45
    Scale
   ) -> (Pose) {
    alphaInputType=Bool
    rotationMode=BMM_Additive
    translationMode=BMM_Additive
    alphaBoolBlend {blendInTime=0.4, blendOption=ExpOut, blendOutTime=25}
    boneToModify {boneName=pack}
    translation {y=-1}
   }
   "Transform (Modify) Bone - Bone: pack"!(
    ComponentPose=@n3
    Alpha:float=Get alpha_rotSpine_inCrouching() -> (alpha_rotSpine_inCrouching:double)
    Translation:Vector="0,7,-2"
    Rotation:Rotator="0,0,-1"
    Scale
   ) -> (Pose) {
    rotationMode=BMM_Additive
    translationMode=BMM_Additive
    alphaBoolBlend {blendInTime=0.2, blendOption=Linear, blendOutTime=25}
    alphaScaleBiasClamp {bInterpResult=true, clampMax=5, clampMin=-5, interpSpeedDecreasing=15, interpSpeedIncreasing=5}
    boneToModify {boneName=pack}
    rotation {roll=-1}
    translation {y=7, z=-2}
   }
   n4 = "Transform (Modify) Bone - Bone: pack"!(
    ComponentPose=Reroute Node(InputPin:Component Space Pose Link Structure) -> (OutputPin:Component Space Pose Link Structure)
    bAlphaBoolEnabled:bool=Get bIsSkydiveFloating()
    Translation="0,5,4"
    Rotation="0,0,-10"
    Scale
   ) -> (Pose) {
    alphaInputType=Bool
    rotationMode=BMM_Additive
    translationMode=BMM_Additive
    alphaBoolBlend {blendInTime=0.1, blendOption=Linear, blendOutTime=1}
    boneToModify {boneName=pack}
    rotation {roll=-10}
    translation {y=5, z=4}
   }
   n5 = "Transform (Modify) Bone - Bone: pack"(ComponentPose=@n4, bAlphaBoolEnabled=Get b_get_out_of_ParaState() -> (b_get_out_of_ParaState:bool), Translation="0,-4,0", Rotation="0,0,-1") {
    alphaInputType=Bool
    rotationMode=BMM_Additive
    translationMode=BMM_Additive
    alphaBoolBlend {blendInTime=0.1, blendOption=Linear, blendOutTime=0.8}
    boneToModify {boneName=pack}
    rotation {roll=-1}
    translation {y=-4}
   }
   n6 = "Transform (Modify) Bone - Bone: pack"(
    ComponentPose=@n5
    bAlphaBoolEnabled=Get Para Glide Open Glider_0() -> (Para Glide Open Glider_0:bool)
    Translation="0,-6,-1"
    Rotation="0,0,10"
   ) {
    alphaInputType=Bool
    rotationMode=BMM_Additive
    translationMode=BMM_Additive
    alphaBoolBlend {blendInTime=0.25, blendOption=Linear, blendOutTime=5}
    alphaScaleBiasClamp {bInterpResult=true}
    boneToModify {boneName=pack}
    rotation {roll=10}
    translation {y=-6, z=-1}
   }
   n7 = "Transform (Modify) Bone - Bone: pack"!(
    ComponentPose=@n6
    bAlphaBoolEnabled=Get bIsParachuteOpened()
    Translation="0,-3,0"
    Rotation_Roll:float=Get!() -> (RotateSpineCond_inSky_max_Roll:float, RotateSpineCond_inSky_max_Pitch:float, RotateSpineCond_inSky_max_Yaw:float).RotateSpineCond_inSky_max_Yaw
    Rotation_Pitch:float=0
    Rotation_Yaw:float=0
    Scale
   ) -> (Pose) {
    alphaInputType=Bool
    rotationMode=BMM_Additive
    translationMode=BMM_Additive
    alphaBoolBlend {blendInTime=0.15, blendOption=Linear, blendOutTime=0.5}
    boneToModify {boneName=pack}
    rotation {roll=-5}
    translation {y=-3}
   }
   n8 = "Transform (Modify) Bone - Bone: pack"!(ComponentPose=@n7, bAlphaBoolEnabled=Get bGreaterVel() -> (bGreaterVel:bool), Translation="0,-4,0", Rotation:Rotator="0,3,-5", Scale) -> (Pose) {
    alphaInputType=Bool
    rotationMode=BMM_Additive
    translationMode=BMM_Additive
    alphaBoolBlend {blendInTime=0.1, blendOption=Linear, blendOutTime=2}
    boneToModify {boneName=pack}
    rotation {roll=-5, yaw=3}
    translation {y=-4}
   }
   n9 = "Transform (Modify) Bone - Bone: pack"!(ComponentPose=@n8, Alpha:float=Get f_KneeSliding_value() -> (f_KneeSliding_value:float), Translation="0,-8,0", Rotation="0,3,-1", Scale) -> (Pose) {
    rotationMode=BMM_Additive
    translationMode=BMM_Additive
    alphaBoolBlend {blendInTime=0.1, blendOption=Linear, blendOutTime=1}
    alphaScaleBiasClamp {bInterpResult=true, interpSpeedDecreasing=2}
    boneToModify {boneName=pack}
    rotation {roll=-1, yaw=3}
    translation {y=-8}
   }
   n10 = "Transform (Modify) Bone - Bone: pack"!(ComponentPose=@n9, bAlphaBoolEnabled:bool=Get bIsDBNO(), Translation="0,10,0", Rotation="0,3,-1", Scale) -> (Pose) {
    alphaInputType=Bool
    rotationMode=BMM_Additive
    translationMode=BMM_Additive
    alphaBoolBlend {blendInTime=0.1, blendOption=Linear, blendOutTime=1}
    boneToModify {boneName=pack}
    rotation {roll=-1, yaw=3}
    translation {y=10}
   }
   "Transform (Modify) Bone - Bone: pack"!(ComponentPose=@n10, Alpha:float, Translation="0,-4,0", Rotation="0,3,-1", Scale) -> (Pose) {
    rotationMode=BMM_Additive
    translationMode=BMM_Additive
    alphaBoolBlend {blendInTime=0.1, blendOption=Linear, blendOutTime=1}
    boneToModify {boneName=pack}
    rotation {roll=-1, yaw=3}
    translation {y=-4}
   }
   n11 = "Transform (Modify) Bone - Bone: general_jnt"!(
    ComponentPose=Local To Component(LocalPose=Use cached pose 'baseDyn_cache'())
    Alpha:float=1
    Translation="0,0,0"
    Rotation=Get GimbalRotation() -> (GimbalRotation:Rotator)
    Scale="1,1,1"
   ) -> (Pose) {
    rotationMode=BMM_Additive
    rotationSpace=BCS_BoneSpace
    translationMode=BMM_Additive
    translationSpace=BCS_BoneSpace
    alphaBoolBlend {blendOption=Linear}
    boneToModify {boneName=general_jnt}
   }
   n12 = "Transform (Modify) Bone - Bone: general_jnt"(ComponentPose=@n11, Alpha=Get rotationBack() -> (rotationBack:double), Translation="0,0,0", Rotation="0,0,-45", Scale="1,1,1") {
    rotationMode=BMM_Additive
    rotationSpace=BCS_BoneSpace
    translationMode=BMM_Additive
    translationSpace=BCS_BoneSpace
    alphaBoolBlend {blendOption=Linear}
    boneToModify {boneName=general_jnt}
    rotation {roll=-45}
   }
   n13 = Get f_swaying_treesSin() -> (f_swaying_treesSin:double)
   n14 = Get scaleZf() -> (scaleZf:double)
   n15 = "Transform (Modify) Bone - Bone: drv_speaker_01_jnt"!(
    ComponentPose="Transform (Modify) Bone - Bone: palm_02_base_02_jnt"!(
     ComponentPose="Transform (Modify) Bone - Bone: palm_01_base_02_jnt"!(ComponentPose=@n12, Alpha=1, Translation="0,0,0", Rotation_Roll:float=0, Rotation_Pitch:float=@n13, Rotation_Yaw:float=0, Scale="1,1,1") -> (Pose) {rotationMode=BMM_Additive, alphaBoolBlend {blendOption=Linear}, boneToModify {boneName=palm_01_base_02_jnt}}
     Alpha=1
     Translation="0,0,0"
     Rotation_Roll:float=0
     Rotation_Pitch:float=@n13
     Rotation_Yaw:float=0
     Scale="1,1,1"
    ) -> (Pose) {rotationMode=BMM_Additive, alphaBoolBlend {blendOption=Linear}, boneToModify {boneName=palm_02_base_02_jnt}}
    Alpha=1
    Translation=Get spk_01_bounce() -> (spk_01_bounce:Vector)
    Rotation="0,0,0"
    Scale_X:double=1
    Scale_Y:double=0
    Scale_Z:double=@n14
   ) -> (Pose) {
    rotationMode=BMM_Additive
    scaleMode=BMM_Additive
    scaleSpace=BCS_BoneSpace
    translationMode=BMM_Additive
    alphaBoolBlend {blendOption=Linear}
    boneToModify {boneName=drv_speaker_01_jnt}
   }
   n16 = Get spk_02_bounce() -> (spk_02_bounce:Vector)
   n17 = "Transform (Modify) Bone - Bone: drv_speaker_02_jnt"!(ComponentPose=@n15, Alpha=1, Translation=@n16, Rotation="0,0,0", Scale_X:double=1, Scale_Y:double=0, Scale_Z:double=@n14) -> (Pose) {
    rotationMode=BMM_Additive
    scaleMode=BMM_Additive
    scaleSpace=BCS_BoneSpace
    translationMode=BMM_Additive
    alphaBoolBlend {blendOption=Linear}
    boneToModify {boneName=drv_speaker_02_jnt}
   }
   n18 = "drv_speaker_02_jnt.scaleX = drv_speaker_02_jnt.scaleZ * -1.3 - Bone Driven Controller"!(ComponentPose:Component Space Pose Link Structure=@n17, Alpha:float=1) -> (Pose:Component Space Pose Link Structure) {
    is=title AnimGraphNode_BoneDrivenController
    bAffectTargetScaleX=true
    modificationMode=AddToRefPose
    multiplier=-1.3
    sourceComponent=ScaleZ
    alphaBoolBlend {blendOption=Linear}
    sourceBone {boneName=drv_speaker_02_jnt}
    targetBone {boneName=drv_speaker_02_jnt}
   }
   n19 = "drv_speaker_01_jnt.scaleY = drv_speaker_01_jnt.scaleZ * -1.3 - Bone Driven Controller"!(ComponentPose:Component Space Pose Link Structure=@n18, Alpha:float=1) -> (Pose:Component Space Pose Link Structure) {
    is=title AnimGraphNode_BoneDrivenController
    bAffectTargetScaleY=true
    modificationMode=AddToRefPose
    multiplier=-1.3
    sourceComponent=ScaleZ
    alphaBoolBlend {blendOption=Linear}
    sourceBone {boneName=drv_speaker_01_jnt}
    targetBone {boneName=drv_speaker_01_jnt}
   }
   n20 = Get bowl_bounce() -> (bowl_bounce:Rotator)
   n21 = "Transform (Modify) Bone - Bone: bowl_b_jnt"(ComponentPose=@n19, Alpha=1, Translation=@n16, Rotation=@n20, Scale="1,1,1") {
    rotationMode=BMM_Additive
    rotationSpace=BCS_BoneSpace
    translationMode=BMM_Additive
    translationSpace=BCS_BoneSpace
    alphaBoolBlend {blendOption=Linear}
    boneToModify {boneName=bowl_b_jnt}
   }
   n22 = "Transform (Modify) Bone - Bone: bone_b_jnt"(ComponentPose=@n21, Alpha=1, Translation="0,0,0", Rotation=@n20, Scale="1,1,1") {
    rotationMode=BMM_Additive
    rotationSpace=BCS_BoneSpace
    translationMode=BMM_Additive
    translationSpace=BCS_BoneSpace
    alphaBoolBlend {blendOption=Linear}
    boneToModify {boneName=bone_b_jnt}
   }
   Output Pose(Result=Component To Local(ComponentPose=@n22)) {layerGroup=None}
   n23 = RigidBody(ComponentPose=Local To Component(LocalPose=Input Pose())) {
    bOverrideWorldGravity=true
    overridePhysicsAsset=&/M_MED_LemonCartGranite_Pack_Physics
    alphaBoolBlend {blendOption=Linear}
    baseBoneRef {boneName=pack}
    componentAppliedLinearAccClamp=<800 800 800>
    componentLinearAccScale=<0.1 0.1 0.1>
    componentLinearVelScale=<1 1 1>
    overrideWorldGravity {z=-100}
    simSpaceSettings {maxAngularAcceleration=800, maxAngularVelocity=3, maxLinearAcceleration=350, maxLinearVelocity=30, velocityScaleZ=0.4, worldAlpha=1}
   }
   n24 = "Constraint - palm_01_base_01_jnt (P)"(ComponentPose=@n23, Alpha=1, ConstraintWeights_0=1) {
    alphaBoolBlend {blendOption=Linear}
    boneToModify {boneName=palm_01_base_01_jnt}
    constraintSetup [{targetBone {boneName=general_jnt}}]
    constraintWeights [1]
   }
   n25 = "Constraint - palm_02_base_01_jnt (P)"(ComponentPose=@n24, Alpha=1, ConstraintWeights_0=1) {
    alphaBoolBlend {blendOption=Linear}
    boneToModify {boneName=palm_02_base_01_jnt}
    constraintSetup [{targetBone {boneName=general_jnt}}]
    constraintWeights [1]
   }
   n26 = "Transform (Modify) Bone - Bone: pack"(ComponentPose=@n25, Translation="0,-1,0", Rotation="0,1,-4") {
    rotationMode=BMM_Additive
    translationMode=BMM_Additive
    alphaBoolBlend {blendInTime=0.1, blendOption=Linear, blendOutTime=1}
    alphaScaleBiasClamp {bInterpResult=true, interpSpeedDecreasing=2}
    boneToModify {boneName=pack}
    rotation {roll=-4, yaw=1}
    translation {y=-1}
   }
   Save cached pose 'baseDyn_cache'(Pose=Component To Local(ComponentPose=@n26)) {cacheName=baseDyn_cache}
   Reroute Node(InputPin=Reroute Node())
  ]}
 EventGraph {nodes [
   n1 = Event BlueprintUpdateAnimation()
   > Set DeltaTimeLocal(DeltaTimeLocal:double=@n1.DeltaTimeX) -> (Output_Get:double)
   > n2 = Sequence!(execute) -> (then_0, then_1, then_2:Exec, then_3:Exec, then_4:Exec, then_5:Exec, then_6:Exec, then_7:Exec, then_8:Exec, then_9:Exec)
   > n3 = Sequence!(execute) -> (then_0, then_1, then_2)
   n4 = GetOwnerSkeletalMeshComponent()
   n5 = MakeRotator(Yaw="Min (Float)"(
     A:double=GetSocketTransform!(self=@n4, InSocketName=spine_04, TransformSpace=RTS_ParentBoneSpace) -> (
      ReturnValue_Location:Vector
      ReturnValue_Rotation_Roll:float
      ReturnValue_Rotation_Pitch:float
      ReturnValue_Rotation_Yaw:float
      ReturnValue_Scale:Vector
     ).ReturnValue_Rotation_Yaw
     B:double=0
    ) -> (ReturnValue:double) {is=title K2Node_CommutativeAssociativeBinaryOperator})
   n6 = Negate Float!(" ":Exec=@n7.then, Value:"Float (double-precision) (by ref)"=BreakRotator(InRot=Invert Rotator!(A:Rotator=@n5) -> (ReturnValue:Rotator)).Yaw) -> ("  ":Exec, Result:double=~) {is=title K2Node_MacroInstance}
   n7 = Set RotateSpineCond_inSky_max(execute=@n3.then_1, RotateSpineCond_inSky_max:Rotator=MakeRotator(Yaw=@n6)) -> (Output_Get:Rotator)
   n8 = GetAlphaFromReferenceTransformDelta(
    SkeletalMeshComponent=@n4
    BoneName=spine_04
    BaseName=spine_05
    DeltaComponentType=QuaternionTwist
    InputRange=(X=-1,Y=10)
    TwistAxis="0,0,1"
   )
   Set b_get_out_of_ParaState(execute=@n3.then_0, b_get_out_of_ParaState:bool=AND Boolean(A=Get bIsSkydiveFloating(), B="float < float"(A=@n8.RawOutput, B=-15))) -> (Output_Get:bool)
   n9 = GetOwnerSkeletalMeshComponent()
   n10 = BreakRotator(InRot=BreakTransform(InTransform=GetDeltaTransformFromRefPose(self=@n9, BoneName=spine_05, BaseName=root)).Rotation)
   > Set GimbalRotation(GimbalRotation:Rotator=MakeRotator(Roll=float * float(A=@n10.Yaw, B=-1) {is=title K2Node_PromotableOperator})) -> (Output_Get:Rotator)
   n11 = Get bIsParachuteOpened()
   n12 = Get BwdMovement() -> (BwdMovement:double)
   > Set rotationBack(rotationBack:double=FInterpTo(
     Current=Get rotationBack()
     Target=float * float(A=@n12, B:bool=@n11)
     DeltaTime=Get DeltaTimeLocal() -> (DeltaTimeLocal:double)
     InterpSpeed=2
    )) -> (Output_Get:double)
   > n13 = PlayerMovement(AnimInstance=Self-Reference(), XDotProductCoefficient=100, YDotProductCoefficient=100, ZDotProductCoefficient=100)
   n14 = BreakVector(InVec=MakeVector(X=@n13, Y=@n13.YMovement))
   > Set LeftMovement(LeftMovement:double="To Float (Boolean)"(InBool="float < float"(A=@n14, B=-50))) -> (Output_Get:double)
   > Set RightMovement(RightMovement:double="To Float (Boolean)"(InBool="float > float"(A=@n14, B=50))) -> (Output_Get:double)
   > Set FwdMovement(FwdMovement:double="To Float (Boolean)"(InBool="float > float"(A=@n14.Y, B=40))) -> (Output_Get:double)
   > Set BwdMovement(BwdMovement:double="To Float (Boolean)"(InBool="float < float"(A=@n14.Y, B=-40))) -> (Output_Get:double)
   n15 = GetOwnerSkeletalMeshComponent(self=Self-Reference())
   n16 = Get World Location!(self:Scene Component Object Reference=@n15) -> (ReturnValue_X:double, ReturnValue_Y:double, ReturnValue_Z:double) {is=title K2Node_CallFunction}
   n17 = Negate Float(" "=@n2.then_3, Value=@n16.ReturnValue_Z) -> (Result=~)
   n18 = SelectFloat(A=@n17, B=@n16.ReturnValue_Z, bPickA="float < float"(A=@n16.ReturnValue_Z))
   n19 = Branch(execute=[@n2.then_6, @n17."  "], Condition="float < float"(A=@n18, B=250))
   n20 = float * float(A=@n16.ReturnValue_Z, B:double=100)
   > Set worldPos_value(worldPos_value:Vector=MakeVector(Y=0, Z=Lerp(A=@n18, B=@n20, Alpha=@n16.ReturnValue_Z))) -> (Output_Get:Vector)
   Set worldPos_value(
    execute=@n19.else
    worldPos_value=MakeVector(Y=0, Z=BreakVector!(InVec_X:double=@n16, InVec_Y:double=@n16.ReturnValue_Y, InVec_Z:double=@n18) -> (X, Y, Z).Z)
   )
   n21 = GetOwnerSkeletalMeshComponent()
   Set alpha_rotSpine_inCrouching(
    execute=@n2.then_5
    alpha_rotSpine_inCrouching:double=GetAlphaFromReferenceTransformDelta(SkeletalMeshComponent=@n21, BoneName=spine_05, BaseName=pelvis, InputRange=(X=35,Y=45), TwistAxis="1,0,0")
   ) -> (Output_Get:double)
   > Set f_KneeSliding_value(f_KneeSliding_value:float=GetAlphaFromReferenceTransformDelta(SkeletalMeshComponent=@n21, BoneName=spine_03, BaseName=pelvis, InputRange=(X=63,Y=69), TwistAxis="1,0,0")) -> (Output_Get:float)
   n22 = Set Para Glide Open Glider(
    execute=@n2.then_2
    Para Glide Open Glider:bool=Get bParaGlide_OpenGlider!(self:Fort Player Anim Instance Object Reference=GetOwnerFortPlayerAnimInstance!(self:Custom Character Part Anim Instance Object Reference) -> (ReturnValue:Fort Player Anim Instance Object Reference)) -> (bParaGlide_OpenGlider:bool)
   ) -> (Output_Get:bool)
   > n23 = Branch(Condition=@n22)
   > Delay(Duration=1)
   > Set Para Glide Open Glider_0(Para Glide Open Glider_0:bool=true) -> (Output_Get:bool)
   Delay(execute=@n23.else, Duration=2)
   > Set Para Glide Open Glider_0(Para Glide Open Glider_0=false)
   n24 = "Max (Float)"(
    A=BreakRotator(InRot=GetSocketTransform!(self=@n15, InSocketName=clavicle_r, TransformSpace=RTS_Component) -> (ReturnValue_Location, ReturnValue_Rotation:Rotator, ReturnValue_Scale).ReturnValue_Rotation).Yaw
    B=0
   )
   n25 = Negate Float(" "=@n2.then_7, Value=float / float(A=@n24, B=5)) -> (Result=~)
   > "Set compRotZ_WhenMeleeAnim (Yaw)"("compRotZ_WhenMeleeAnim (Yaw)":float=@n25) -> (Output_Get:float)
   n26 = MapRangeClamped(Value=@n24, InRangeA=60, InRangeB=100, OutRangeB=6)
   > n27 = Negate Float(Value=@n26) -> (Result=~)
   n28 = Get bIsPlayingMeleeAnim()
   n29 = Get bIsCrouchSprinting()
   > Set compTransfWhenMeleeAnimZ(compTransfWhenMeleeAnimZ:double=SelectFloat(A=float * float(A=@n26, B=2), B=@n26, bPickA=AND Boolean(A=@n29, B=@n28))) -> (Output_Get:double)
   > Set compTransfWhenMeleeAnimY(compTransfWhenMeleeAnimY:double=float + float(A=@n27, B=-2)) -> (Output_Get:double)
   n30 = "Max (Float)"(A=Get PawnVelocityZ(), B=0)
   n31 = Get bIsParachuteOpened()
   n32 = Branch(execute=@n2.then_1, Condition=AND Boolean(A="float > float"(A=@n30, B=3000), B="Not Equal (Boolean)"(A=@n31, B=true) -> (ReturnValue=~)))
   > Set bGreaterVel(bGreaterVel:bool=true) -> (Output_Get:bool)
   Set bGreaterVel(execute=@n32.else, bGreaterVel=false)
   n33 = GetGameTimeInSeconds()
   n34 = Branch(
    execute=@n2.then_9
    Condition="Nearly Equal (Float)"(A="% (Float)"!(A:double=@n33, B:double=0.43) -> (ReturnValue:double), B=0, ErrorTolerance=0.26)
   )
   n35 = float * float(
    A=RandomFloatFromStream!(Stream:"Random Stream Structure (by ref)"=Get random_stream() -> (random_stream:Random Stream Structure)) -> (ReturnValue:float)
    B=0.9
   ) {is=title K2Node_CommutativeAssociativeBinaryOperator}
   n36 = "Sin (Radians)"(A=float * float(A="Clamp (Float)"(Value=@n35, Min=0.5, Max=0.8), B=GetGameTimeInSeconds()))
   n37 = float * float(A=@n36, B=0.08)
   n38 = MakeVector(Y=@n37)
   > Set spk_02_bounce(spk_02_bounce:Vector=@n38) -> (Output_Get:Vector)
   > Set spk_02_bounce(spk_02_bounce=@n38)
   > Set spk_01_bounce(spk_01_bounce:Vector=MakeVector(X=@n37, Y=0)) -> (Output_Get:Vector)
   > Set bowl_bounce(bowl_bounce:Rotator=MakeRotator(Pitch=float * float(A=@n36, B=4))) -> (Output_Get:Rotator)
   > Set scaleZf(scaleZf:double=MapRangeClamped(Value="Absolute (Float)"(A=@n36), InRangeA=0, InRangeB=1, OutRangeA=0.9, OutRangeB=1.05)) -> (Output_Get:double)
   > Set scaleYf(scaleYf:double=1) -> (Output_Get:double)
   Set spk_02_bounce(execute=@n34.else, spk_02_bounce="0, 0, 0")
   n39 = "Sin (Radians)"(A=float * float(A=@n33, B=2) {is=title K2Node_PromotableOperator})
   Set f_swaying_treesSin(execute=@n2.then_4, f_swaying_treesSin:double=float * float(A="% (Float)"(A=@n39, B=1), B=3)) -> (Output_Get:double)
   Event BlueprintInitializeAnimation()
   > n40 = Cast To FortPlayerPawn(Object=TryGetFortPawn())
   > Set Skeletal Mesh(Skeletal Mesh:Skeletal Mesh Component Object Reference=GetSkeletalMeshForPartType(self=@n40)) -> (Output_Get:Skeletal Mesh Component Object Reference)
  ]}
]
variables [
 Alpha Output {type=number}
 alpha_rotSpine_inCrouching {type=number}
 b_get_out_of_ParaState {type=boolean}
 bGreaterVel {type=boolean}
 bone_b_rot_bounce {type=Rotator}
 bowl_b_loc_bounce {type=Vector}
 bowl_bounce {type=Rotator}
 BwdMovement {type=number}
 "compRotZ_WhenMeleeAnim (Yaw)" {type=number}
 compTransfWhenMeleeAnimY {type=number}
 compTransfWhenMeleeAnimZ {type=number}
 DeltaTimeLocal {type=number}
 f_KneeSliding_value {type=number}
 f_swaying_treesSin {type=number}
 FwdMovement {type=number}
 GimbalRotation {type=Rotator}
 LeftMovement {type=number}
 NewVar {type=number}
 Output {type=SkeletalMeshComponentBudgeted}
 Para Glide Open Glider {type=boolean}
 Para Glide Open Glider_0 {type=boolean}
 random_stream {type=RandomStream, default {initialSeed=0, seed=0}}
 RightMovement {type=number}
 RotateSpineCond_inSky_max {type=Rotator}
 rotationBack {type=number}
 scale_01Vec {type=Vector}
 scale_02Vec {type=Vector}
 scaleYf {type=number}
 scaleZf {type=number}
 Skel_Mesh {type=SkeletalMeshComponent}
 Skeletal Mesh {type=SkeletalMeshComponent}
 spk_01_bounce {type=Vector}
 spk_02_bounce {type=Vector}
 spk_loc_bounce_1 {type=Vector}
 spk_loc_bounce_2 {type=Vector}
 spk_sc_bounce_1 {type=Vector}
 spk_sc_bounce_2 {type=Vector}
 worldPos_value {type=Vector}
]
